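# Physics asset for the Moss Knight cosmetic backpack, defining collision bodies and joints for its dangling flowers, leaves, ferns and vines.
# Flowers, standalone leaves and the first segments of two fern chains and two vine chains hang off the pack capsule.
# Leaves branch from fern and bottom-vine segments, with angular limits tuned per plant part.
# The pack joins a three-capsule root body with free swing and twist.
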PhysicsAsset /BRCosmetics/Accessories/FORT_Backpacks/Backpack_M_MED_Moss_Knight/Meshes/M_MED_Moss_Knight_Pack_Physics
bodies {
 dyn_flower_01 [{shapeType=Box, extentX=0.7479, extentY=3.2204, extentZ=3.5128, center=<0.2793 0 0>}]
 dyn_flower_02 [{shapeType=Box, extentX=0.7479, extentY=3.2204, extentZ=3.5128, center=<0.2793 0 0>}]
 dyn_flower_03 [{shapeType=Capsule, length=1.7607, radius=1.7295, center=<0.1993 0.1659 0.0978>, rotation=<rot -0.564 -90 0.0435>}]
 dyn_flower_04 [{shapeType=Box, extentX=0.7479, extentY=3.2204, extentZ=3.5128, center=<0.2793 0 0>}]
 dyn_l_fern_01 [{shapeType=Capsule, length=2.8463, radius=0.7203, center=<2.1429 0 0>, rotation=<rot 90 0 0>}]
 dyn_l_fern_02 [{shapeType=Capsule, length=2.8463, radius=0.7203, center=<2.1429 0 0>, rotation=<rot 90 0 0>}]
 dyn_l_fern_03 [{shapeType=Capsule, length=2.8463, radius=0.7203, center=<2.1429 0 0>, rotation=<rot 90 0 0>}]
 dyn_l_fern_04 [{shapeType=Capsule, length=2.8463, radius=0.7203, center=<2.1429 0 0>, rotation=<rot 90 0 0>}]
 dyn_l_fern_leaf_01 [{shapeType=Capsule, length=3.6823, radius=1.3393, center=<3.5134 -0.036 -0.0675>, rotation=<rot 0 -87.8801 -89.0989>}]
 dyn_l_fern_leaf_02 [{shapeType=Capsule, length=4.7378, radius=1.3818, center=<5.137 -0.1361 -0.3769>, rotation=<rot 0 -94.0296 -80.7364>}]
 dyn_l_fern_leaf_03 [{shapeType=Capsule, length=3.7486, radius=1.117, center=<3.6467 0.1403 0.2444>, rotation=<rot 0 -90.9362 -87.6399>}]
 dyn_l_fern_leaf_04 [{shapeType=Capsule, length=6.4819, radius=1.3388, center=<6.5118 -0.2577 0.2938>, rotation=<rot 0 -93.923 -84.5955>}]
 dyn_l_fern_leaf_05 [{shapeType=Capsule, length=6.4229, radius=1.4392, center=<6.1786 0.0139 -2.6684>, rotation=<rot 0 -91.6226 -57.8559>}]
 dyn_l_fern_leaf_06 [{shapeType=Capsule, length=6.9855, radius=1.4088, center=<6.8815 0.5333 0.4023>, rotation=<rot 0 -87.5403 -87.9201>}]
 dyn_l_fern_leaf_07 [{shapeType=Capsule, length=7.8882, radius=1.5914, center=<8.167 -0.2724 -0.2802>, rotation=<rot 0 -90.336 -82.6294>}]
 dyn_l_leaf_bot [{shapeType=Capsule, length=2.5302, radius=1.2574, center=<2.8351 -0.1016 -0.464>, rotation=<rot 0 -91.4818 -79.0833>}]
 dyn_l_leaf_mid [{shapeType=Capsule, length=2.6989, radius=1.401, center=<2.6228 0.0453 0.0829>, rotation=<rot 0 -90.3669 -85.1882>}]
 dyn_l_leaf_top [{shapeType=Capsule, length=2.8796, radius=2.124, center=<2.9408 0.1227 -0.4998>, rotation=<rot 0 -90.3577 -76.5062>}]
 dyn_r_fern_01 [{shapeType=Capsule, length=2.8463, radius=0.7203, center=<2.1429 0 0>, rotation=<rot 90 0 0>}]
 dyn_r_fern_02 [{shapeType=Capsule, length=2.8463, radius=0.7203, center=<2.1429 0 0>, rotation=<rot 90 0 0>}]
 dyn_r_fern_03 [{shapeType=Capsule, length=2.8463, radius=0.7203, center=<2.1429 0 0>, rotation=<rot 90 0 0>}]
 dyn_r_fern_04 [{shapeType=Capsule, length=2.8463, radius=0.7203, center=<2.1429 0 0>, rotation=<rot 90 0 0>}]
 dyn_r_fern_leaf_01 [{shapeType=Box, extentX=5.816, extentY=2, extentZ=1.2499, center=<3.3081 0 0.0001>}]
 dyn_r_fern_leaf_02 [{shapeType=Box, extentX=5.816, extentY=2, extentZ=1.2499, center=<3.3081 0 0.0001>}]
 dyn_r_fern_leaf_03 [{shapeType=Box, extentX=7.3255, extentY=2, extentZ=1.2499, center=<4.0024 0 0.0001>}]
 dyn_r_fern_leaf_04 [{shapeType=Box, extentX=11.6169, extentY=2, extentZ=1.2499, center=<5.8718 0 0.0001>}]
 dyn_r_fern_leaf_05 [{shapeType=Box, extentX=7.6187, extentY=2, extentZ=1.2499, center=<3.8713 0 0.0001>}]
 dyn_r_fern_leaf_06 [{shapeType=Box, extentX=15.5047, extentY=2.7013, extentZ=1.2499, center=<8.0948 0 0.0001>}]
 dyn_r_fern_leaf_07 [{shapeType=Box, extentX=8.6901, extentY=2, extentZ=1.2499, center=<4.7168 0 0.0001>}]
 dyn_top_leaf_bot [{shapeType=Capsule, length=3.0636, radius=0.9364, center=<3.3068 0 0>, rotation=<rot 90 0 0>}]
 dyn_top_leaf_top [{shapeType=Capsule, length=3.0636, radius=0.9364, center=<3.3068 0 0>, rotation=<rot 90 0 0>}]
 dyn_vine_bot_01 [{shapeType=Capsule, length=3.0677, radius=0.9323, center=<2.5331 0 -0.0002>, rotation=<rot 90 0 0>}]
 dyn_vine_bot_02 [{shapeType=Capsule, length=5.6776, radius=1.106, center=<4.0566 0 -0.0001>, rotation=<rot 90 0 0>}]
 dyn_vine_bot_03 [{shapeType=Capsule, length=3.1435, radius=0.8565, center=<2.802 0 -0.0001>, rotation=<rot 90 0 0>}]
 dyn_vine_bot_04 [{shapeType=Capsule, length=3.0872, radius=0.7456, center=<2.1715 0 -0.0001>, rotation=<rot 90 0 0>}]
 dyn_vine_bot_05 [{shapeType=Capsule, length=3.2747, radius=0.7253, center=<2.437 0 -0.0002>, rotation=<rot 90 0 0>}]
 dyn_vine_bot_06 [{shapeType=Capsule, length=3.3269, radius=0.6732, center=<2.3491 0 -0.0002>, rotation=<rot 90 0 0>}]
 dyn_vine_bot_07 [{shapeType=Capsule, length=2.7608, radius=0.6053, center=<1.935 0 -0.0002>, rotation=<rot 90 0 0>}]
 dyn_vine_bot_08 [{shapeType=Capsule, length=2.3307, radius=0.704, center=<1.9348 0 -0.0002>, rotation=<rot 90 0 0>}]
 dyn_vine_bot_09 [{shapeType=Capsule, length=3.3477, radius=0.6523, center=<2.432 0 -0.0002>, rotation=<rot 90 0 0>}]
 dyn_vine_bot_10 [{shapeType=Capsule, length=3.131, radius=0.8691, center=<2.5331 0 -0.0002>, rotation=<rot 90 0 0>}]
 dyn_vine_bot_leaf_01 [{shapeType=Capsule, length=2.0761, radius=1.4397, center=<2.255 0.0247 -0.0043>, rotation=<rot 0 -89.3561 -85.4151>}]
 dyn_vine_bot_leaf_02 [{shapeType=Capsule, length=2.128, radius=1.4494, center=<2.111 -0.2001 -0.0088>, rotation=<rot 0 -96.4685 -83.3349>}]
 dyn_vine_bot_leaf_03 [{shapeType=Capsule, length=1.7778, radius=1.2134, center=<1.7781 0.0718 0.1171>, rotation=<rot 0 -89.9635 -86.8527>}]
 dyn_vine_l_01 [{shapeType=Capsule, length=2.0823, radius=0.8681, center=<1.8467 0 0.0002>, rotation=<rot 90 0 0>}]
 dyn_vine_l_02 [{shapeType=Capsule, length=2.7673, radius=0.8681, center=<2.2875 0 0.0002>, rotation=<rot 90 0 0>}]
 dyn_vine_l_03 [{shapeType=Capsule, length=2.2378, radius=0.7126, center=<1.8467 0 0.0002>, rotation=<rot 90 0 0>}]
 dyn_vine_l_04 [{shapeType=Capsule, length=2.3799, radius=0.5705, center=<1.8467 0 0.0002>, rotation=<rot 90 0 0>}]
 dyn_vine_l_05 [{shapeType=Capsule, length=2.4087, radius=0.5417, center=<1.8467 0 0.0002>, rotation=<rot 90 0 0>}]
 dyn_vine_l_06 [{shapeType=Capsule, length=1.7276, radius=0.5002, center=<1.3906 0 0.0002>, rotation=<rot 90 0 0>}]
 dyn_vine_l_07 [{shapeType=Capsule, length=2.1191, radius=0.5589, center=<1.5454 0 0.0002>, rotation=<rot 90 0 0>}]
 dyn_vine_l_08 [{shapeType=Capsule, length=2.0823, radius=0.8681, center=<1.8467 0 0.0002>, rotation=<rot 90 0 0>}]
 pack [{shapeType=Capsule, length=36.2699, radius=6.0528, center=<0 -0.2226 5.0391>, rotation=<rot 90 0 0>}]
 root [
  {shapeType=Capsule, length=6.3388, radius=3.5705, center=<13.0277 1.9423 -8.4858>, rotation=<rot 90 0 -29.4155>}
  {shapeType=Capsule, length=39.2079, radius=7.1779, center=<0 6.2232 -38.0061>, rotation=<rot 90 0 0>}
  {shapeType=Capsule, length=39.2079, radius=7.1779, center=<0.2265 6.2232 -52.7646>, rotation=<rot 90 0 0>}
 ]
}
constraints [
 {bone1Name=dyn_vine_bot_03, bone2Name=dyn_vine_bot_02, swing1LimitDegrees=20, swing2LimitDegrees=20, twistLimitDegrees=20}
 {bone1Name=dyn_l_fern_03, bone2Name=dyn_l_fern_02, swing1LimitDegrees=5, swing2LimitDegrees=10, twistLimitDegrees=10}
 {bone1Name=dyn_r_fern_03, bone2Name=dyn_r_fern_02, swing1LimitDegrees=5, swing2LimitDegrees=10, twistLimitDegrees=10}
 {bone1Name=dyn_flower_02, bone2Name=pack, swing1LimitDegrees=15, swing2LimitDegrees=15, twistLimitDegrees=15}
 {bone1Name=dyn_flower_03, bone2Name=pack, swing1LimitDegrees=15, swing2LimitDegrees=15, twistLimitDegrees=15}
 {bone1Name=dyn_flower_04, bone2Name=pack, swing1LimitDegrees=15, swing2LimitDegrees=15, twistLimitDegrees=15}
 {bone1Name=dyn_flower_01, bone2Name=pack, swing1LimitDegrees=15, swing2LimitDegrees=15, twistLimitDegrees=15}
 {bone1Name=dyn_top_leaf_bot, bone2Name=pack, swing1LimitDegrees=25, swing2LimitDegrees=25, twistLimitDegrees=5}
 {bone1Name=dyn_top_leaf_top, bone2Name=pack, swing1LimitDegrees=25, swing2LimitDegrees=25, twistLimitDegrees=5}
 {bone1Name=dyn_l_leaf_bot, bone2Name=pack, swing1LimitDegrees=25, swing2LimitDegrees=25, twistLimitDegrees=5}
 {bone1Name=dyn_l_leaf_mid, bone2Name=pack, swing1LimitDegrees=25, swing2LimitDegrees=25, twistLimitDegrees=5}
 {bone1Name=dyn_l_leaf_top, bone2Name=pack, swing1LimitDegrees=25, swing2LimitDegrees=25, twistLimitDegrees=5}
 {bone1Name=dyn_r_fern_01, bone2Name=pack, swing1LimitDegrees=5, swing2LimitDegrees=5, twistLimitDegrees=5}
 {bone1Name=dyn_r_fern_02, bone2Name=dyn_r_fern_01, swing1LimitDegrees=5, swing2LimitDegrees=10, twistLimitDegrees=10}
 {bone1Name=dyn_r_fern_04, bone2Name=dyn_r_fern_03, swing1LimitDegrees=5, swing2LimitDegrees=10, twistLimitDegrees=10}
 {bone1Name=dyn_r_fern_leaf_03, bone2Name=dyn_r_fern_03, swing1LimitDegrees=15, swing2LimitDegrees=30, twistLimitDegrees=5}
 {bone1Name=dyn_r_fern_leaf_04, bone2Name=dyn_r_fern_03, swing1LimitDegrees=15, swing2LimitDegrees=30, twistLimitDegrees=5}
 {bone1Name=dyn_r_fern_leaf_05, bone2Name=dyn_r_fern_03, swing1LimitDegrees=15, swing2LimitDegrees=30, twistLimitDegrees=5}
 {bone1Name=dyn_r_fern_leaf_06, bone2Name=dyn_r_fern_03, swing1LimitDegrees=15, swing2LimitDegrees=30, twistLimitDegrees=5}
 {bone1Name=dyn_r_fern_leaf_01, bone2Name=dyn_r_fern_02, swing1LimitDegrees=15, swing2LimitDegrees=30, twistLimitDegrees=5}
 {bone1Name=dyn_r_fern_leaf_02, bone2Name=dyn_r_fern_02, swing1LimitDegrees=15, swing2LimitDegrees=30, twistLimitDegrees=5}
 {bone1Name=dyn_l_fern_01, bone2Name=pack, swing1LimitDegrees=5, swing2LimitDegrees=5, twistLimitDegrees=5}
 {bone1Name=dyn_l_fern_02, bone2Name=dyn_l_fern_01, swing1LimitDegrees=5, swing2LimitDegrees=10, twistLimitDegrees=10}
 {bone1Name=dyn_l_fern_04, bone2Name=dyn_l_fern_03, swing1LimitDegrees=5, swing2LimitDegrees=10, twistLimitDegrees=10}
 {bone1Name=dyn_l_fern_leaf_04, bone2Name=dyn_l_fern_03, swing1LimitDegrees=15, swing2LimitDegrees=30, twistLimitDegrees=5}
 {bone1Name=dyn_l_fern_leaf_05, bone2Name=dyn_l_fern_03, swing1LimitDegrees=15, swing2LimitDegrees=30, twistLimitDegrees=5}
 {bone1Name=dyn_l_fern_leaf_06, bone2Name=dyn_l_fern_03, swing1LimitDegrees=15, swing2LimitDegrees=30, twistLimitDegrees=5}
 {bone1Name=dyn_l_fern_leaf_07, bone2Name=dyn_l_fern_03, swing1LimitDegrees=15, swing2LimitDegrees=30, twistLimitDegrees=5}
 {bone1Name=dyn_l_fern_leaf_01, bone2Name=dyn_l_fern_02, swing1LimitDegrees=15, swing2LimitDegrees=30, twistLimitDegrees=5}
 {bone1Name=dyn_l_fern_leaf_02, bone2Name=dyn_l_fern_02, swing1LimitDegrees=15, swing2LimitDegrees=30, twistLimitDegrees=5}
 {bone1Name=dyn_l_fern_leaf_03, bone2Name=dyn_l_fern_02, swing1LimitDegrees=15, swing2LimitDegrees=30, twistLimitDegrees=5}
 {bone1Name=dyn_vine_bot_01, bone2Name=pack, swing1LimitDegrees=40, swing2LimitDegrees=30, twistLimitDegrees=30}
 {bone1Name=dyn_vine_bot_02, bone2Name=dyn_vine_bot_01, swing1LimitDegrees=30, swing2LimitDegrees=30, twistLimitDegrees=30}
 {bone1Name=dyn_vine_bot_04, bone2Name=dyn_vine_bot_03, swing1LimitDegrees=20, swing2LimitDegrees=20, twistLimitDegrees=20}
 {bone1Name=dyn_vine_bot_05, bone2Name=dyn_vine_bot_04, swing1LimitDegrees=20, swing2LimitDegrees=20, twistLimitDegrees=20}
 {bone1Name=dyn_vine_bot_06, bone2Name=dyn_vine_bot_05, swing1LimitDegrees=20, swing2LimitDegrees=20, twistLimitDegrees=20}
 {bone1Name=dyn_vine_bot_07, bone2Name=dyn_vine_bot_06, swing1LimitDegrees=20, swing2LimitDegrees=20, twistLimitDegrees=20}
 {bone1Name=dyn_vine_bot_08, bone2Name=dyn_vine_bot_07, swing1LimitDegrees=20, swing2LimitDegrees=20, twistLimitDegrees=20}
 {bone1Name=dyn_vine_bot_09, bone2Name=dyn_vine_bot_08, swing1LimitDegrees=20, swing2LimitDegrees=20, twistLimitDegrees=20}
 {bone1Name=dyn_vine_bot_10, bone2Name=dyn_vine_bot_09, swing1LimitDegrees=20, swing2LimitDegrees=20, twistLimitDegrees=20}
 {bone1Name=dyn_vine_bot_leaf_01, bone2Name=dyn_vine_bot_03, swing1LimitDegrees=25, swing2LimitDegrees=45, twistLimitDegrees=5}
 {bone1Name=dyn_vine_bot_leaf_02, bone2Name=dyn_vine_bot_03, swing1LimitDegrees=25, swing2LimitDegrees=45, twistLimitDegrees=5}
 {bone1Name=dyn_vine_l_08, bone2Name=dyn_vine_l_07, swing1LimitDegrees=20, swing2LimitDegrees=20, twistLimitDegrees=20}
 {bone1Name=dyn_vine_l_01, bone2Name=pack, swing1LimitDegrees=20, swing2LimitDegrees=20, twistLimitDegrees=20}
 {bone1Name=dyn_vine_l_02, bone2Name=dyn_vine_l_01, swing1LimitDegrees=20, swing2LimitDegrees=20, twistLimitDegrees=20}
 {bone1Name=dyn_vine_l_03, bone2Name=dyn_vine_l_02, swing1LimitDegrees=20, swing2LimitDegrees=20, twistLimitDegrees=20}
 {bone1Name=dyn_vine_l_04, bone2Name=dyn_vine_l_03, swing1LimitDegrees=20, swing2LimitDegrees=20, twistLimitDegrees=20}
 {bone1Name=dyn_vine_l_05, bone2Name=dyn_vine_l_04, swing1LimitDegrees=20, swing2LimitDegrees=20, twistLimitDegrees=20}
 {bone1Name=dyn_vine_l_06, bone2Name=dyn_vine_l_05, swing1LimitDegrees=20, swing2LimitDegrees=20, twistLimitDegrees=20}
 {bone1Name=dyn_vine_l_07, bone2Name=dyn_vine_l_06, swing1LimitDegrees=20, swing2LimitDegrees=20, twistLimitDegrees=20}
 {bone1Name=dyn_vine_bot_leaf_03, bone2Name=dyn_vine_bot_06, swing1LimitDegrees=25, swing2LimitDegrees=45, twistLimitDegrees=5}
 {bone1Name=dyn_r_fern_leaf_07, bone2Name=dyn_r_fern_04, swing1LimitDegrees=15, swing2LimitDegrees=30, twistLimitDegrees=5}
 {
  bone1Name=pack
  bone2Name=root
  swing1LimitDegrees=45
  swing1Motion=Free
  swing2LimitDegrees=45
  swing2Motion=Free
  twistLimitDegrees=45
  twistMotion=Free
 }
]
