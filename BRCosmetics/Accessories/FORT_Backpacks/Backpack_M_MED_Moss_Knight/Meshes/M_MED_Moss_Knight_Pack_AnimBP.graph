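# Animation Blueprint for the Moss Knight backpack cosmetic, posing the pack and simulating its dynamic bones, with special handling for skydiving.
# Each update flags freefall (skydiving, parachute closed), switching gravity from -Z to -Y.
# The pack bone gets a constant roll, plus an extra offset and roll in freefall, before a root-space RigidBody simulation.
# Afterwards, position constraints tie three vine bones to the pack during freefall only.
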AnimBlueprint /BRCosmetics/Accessories/FORT_Backpacks/Backpack_M_MED_Moss_Knight/Meshes/M_MED_Moss_Knight_Pack_AnimBP
target_skeleton=&/M_MED_Moss_Knight_Pack_Skeleton
graphs [
 AnimGraph {nodes [
   n1 = "Transform (Modify) Bone - Bone: pack"(ComponentPose=Local To Component(LocalPose=Local Space Ref Pose()), Translation="0,-2,0", Rotation="0,0,10") {
    rotationMode=BMM_Additive
    scaleMode=BMM_Additive
    translationMode=BMM_Additive
    alphaBoolBlend {blendOption=Linear}
    boneToModify {boneName=pack}
    rotation {roll=10}
    translation {y=-2}
   }
   n2 = Component To Local(ComponentPose=@n1)
   n3 = Local Space Ref Pose()
   n4 = Get skydiving() -> (skydiving:bool)
   n5 = "Transform (Modify) Bone - Bone: pack"(
    ComponentPose=Local To Component(LocalPose=Blend Poses by bool(BlendPose_0=@n2, BlendPose_1=@n3, bActiveValue=@n4) {blendType=Linear})
    Translation="0,0,0"
    Rotation="0,0,15"
   ) {
    rotationMode=BMM_Additive
    scaleMode=BMM_Additive
    translationMode=BMM_Additive
    alphaBoolBlend {blendOption=Linear}
    boneToModify {boneName=pack}
    rotation {roll=15}
   }
   n6 = RigidBody!(ComponentPose=@n5, Alpha, OverrideWorldGravity:Vector=Get override_gravity() -> (override_gravity:Vector), ExternalForce) -> (Pose) {
    bClampLinearTranslationLimitToRefPose=true
    bOverrideWorldGravity=true
    overridePhysicsAsset=&/M_MED_Moss_Knight_Pack_Physics
    simulationSpace=BaseBoneSpace
    alphaBoolBlend {blendOption=Linear}
    baseBoneRef {boneName=root}
    componentAppliedLinearAccClamp=<5000 5000 5000>
    componentLinearAccScale=<0.3 0.3 0.4>
    overrideWorldGravity {z=-1100}
   }
   n7 = Get skydiving_float() -> (skydiving_float:double)
   n8 = "Constraint - dyn_vine_bot_01 (P)"(ComponentPose=@n6, Alpha=@n7, ConstraintWeights_0=1) {
    alphaBoolBlend {blendOption=Linear}
    boneToModify {boneName=dyn_vine_bot_01}
    constraintSetup [{targetBone {boneName=pack}}]
    constraintWeights [1]
   }
   n9 = "Constraint - dyn_vine_bot_02 (P)"(ComponentPose=@n8, Alpha=@n7, ConstraintWeights_0=1) {
    alphaBoolBlend {blendOption=Linear}
    boneToModify {boneName=dyn_vine_bot_02}
    constraintSetup [{targetBone {boneName=pack}}]
    constraintWeights [1]
   }
   n10 = "Constraint - dyn_vine_l_01 (P)"(ComponentPose=@n9, Alpha=@n7, ConstraintWeights_0=1) {
    alphaBoolBlend {blendOption=Linear}
    boneToModify {boneName=dyn_vine_l_01}
    constraintSetup [{targetBone {boneName=pack}}]
    constraintWeights [1]
   }
   Output Pose(Result=Component To Local(ComponentPose=@n10)) {layerGroup=None}
  ]}
 EventGraph {nodes [
   Event BlueprintUpdateAnimation()
   n1 = Get bIsParachuteOpened()
   > n2 = Branch(Condition=AND Boolean(A=NOT Boolean(A=@n1), B=Get bIsSkyDiving()))
   > Set override_gravity(override_gravity:Vector="0,-1100, 0") -> (Output_Get:Vector)
   > Set skydiving(skydiving:bool=true) -> (Output_Get:bool)
   > Set skydiving_float(skydiving_float:double=1) -> (Output_Get:double)
   Set override_gravity(execute=@n2.else, override_gravity="0, 0,-1100")
   > Set skydiving(skydiving=false)
   > Set skydiving_float(skydiving_float=0)
   TryGetPawnOwner()
  ]}
]
variables [
 override_gravity {type=Vector}
 skydiving {type=boolean}
 skydiving_float {type=number}
]
